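# Stereo MetaSound source for first-person emote music, built from a randomly picked intro wave and a randomly picked, optionally looping loop wave.
# DelayLoopByStart selects whether the loop fires on play or when the intro finishes or is missing.
# Both players share a pitch multiplier converted to semitones and mix through separate volume multipliers scaled by a fade-in envelope.
# IntroDelay is unused, and the combined finish trigger drives no output.
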MetaSoundSource /FortMetasound/Cosmetics/Emote/MSS_Emote_Music1P
inputs [
 DelayLoopByStart {type=Bool, access=Constructor}
 FadeInTime {type=Time, access=Constructor}
 IntroDelay {type=Time, access=Constructor}
 IntroVolumeMultiplier {type=Float}
 Looping {type=Bool, access=Constructor}
 LoopVolumeMultiplier {type=Float}
 Music.Loop.Sounds {type=WaveAsset:Array, access=Constructor}
 Music.Sounds {type=WaveAsset:Array, access=Constructor}
 PitchTemplateMultiplier {type=Float}
 UE.Source.OnPlay {type=Trigger}
 UE.Source.StartTime.StartTime {type=Time}
]
interfaces [UE.OutputFormat.Stereo:1.0, UE.Source.StartTime:1.0, UE.Source:1.1]
outputs [
 "UE.OutputFormat.Stereo.Audio:0" {type=Audio}
 "UE.OutputFormat.Stereo.Audio:1" {type=Audio}
]
pages [Default {
  input_defaults {
   DelayLoopByStart=null
   FadeInTime=0
   IntroDelay=4.114
   IntroVolumeMultiplier=1
   LoopVolumeMultiplier=1
   Looping="<MetaSoundBuilderResult.SUCCEEDED: 0>"
   PitchTemplateMultiplier=1
   UE.Source.OnPlay=null
   UE.Source.StartTime.StartTime=0
   Music.Loop.Sounds []
   Music.Sounds []
  }
  nodes [
   ADSREnvelopeADSREnvelope_0 = ADSR Envelope.ADSR Envelope:Float(Attack Time=^FadeInTime, Trigger Attack=^UE.Source.OnPlay) {also {defaults {Sustain Level=1}}}
   n1 = VariableAccessor.Float(Variable=VariableMutator.Float(Value=@ADSREnvelopeADSREnvelope_0.Out Envelope, Variable=InitVariable.Float().Variable).Variable)
   n2 = InitVariable.Audio()
   n3 = /FortMetasound/Patches/IsWaveAssetArrayValid(WaveAsset=^Music.Sounds, Trigger In=^UE.Source.OnPlay)
   ArrayRandomGet_0 = Array.Random Get:WaveAsset:Array(In Array=^Music.Sounds, Next=@n3.Is Valid)
   FrequencyMultipliertoSemitone_1 = UE.Frequency Multiplier to Semitone:Float(Frequency Multiplier=^PitchTemplateMultiplier)
   WavePlayer_0 = UE.Wave Player:Stereo(
    Start Time=^UE.Source.StartTime.StartTime
    Play=@ArrayRandomGet_0.On Next
    Wave Asset=UE.Reroute(Value=UE.Reroute(Value=@ArrayRandomGet_0.Value).Value).Value
    Pitch Shift=UE.Reroute(Value=@FrequencyMultipliertoSemitone_1.Semitones).Value
   ) {also {version [1,1]}}
   n4 = InitVariable.Audio()
   n5 = InitVariable.Trigger()
   TriggerAnyTriggerAny2_2 = "TriggerAny.Trigger Any (2)"(In 0=UE.Reroute(Value=@n3.Is Not Valid).Value, In 1=@WavePlayer_0.On Finished)
   n6 = VariableAccessor.Trigger(Variable=VariableMutator.Trigger(Variable=InitVariable.Trigger().Variable, Value=@TriggerAnyTriggerAny2_2.Out).Variable)
   TriggerAnyTriggerAny2_0 = "TriggerAny.Trigger Any (2)"(
    In 0=TriggerCompare.Clamp:Bool(A=^DelayLoopByStart, Compare=^UE.Source.OnPlay).TRUE
    In 1=TriggerCompare.Clamp:Bool(A=^DelayLoopByStart, Compare=@n6.Value) {also {defaults {B="<MetaSoundBuilderResult.SUCCEEDED: 0>"}}}.TRUE
   )
   n7 = &IsWaveAssetArrayValid(
    WaveAsset=^Music.Loop.Sounds
    Trigger In=VariableAccessor.Trigger(Variable=VariableMutator.Trigger(Variable=@n5.Variable, Value=@TriggerAnyTriggerAny2_0.Out).Variable).Value
   )
   ArrayRandomGet_1 = Array.Random Get:WaveAsset:Array(Next=@n7.Is Valid, In Array=UE.Reroute(Value=^Music.Loop.Sounds).Value)
   FrequencyMultipliertoSemitone_0 = UE.Frequency Multiplier to Semitone:Float(Frequency Multiplier=^PitchTemplateMultiplier)
   WavePlayer_1 = UE.Wave Player:Stereo(
    Loop=^Looping
    Start Time=^UE.Source.StartTime.StartTime
    Play=@ArrayRandomGet_1.On Next
    Wave Asset=@ArrayRandomGet_1.Value
    Pitch Shift=@FrequencyMultipliertoSemitone_0.Semitones
   ) {also {version [1,1]}}
   n8 = InitVariable.Audio()
   n9 = InitVariable.Audio()
   AudioMixerAudioMixerStereo2_0 = "AudioMixer.Audio Mixer (Stereo, 2)"(
    Gain 1=UE.Multiply:Float(AdditionalOperands=^IntroVolumeMultiplier, PrimaryOperand=@n1.Value).Out
    Gain 0=UE.Multiply:Float(AdditionalOperands=^LoopVolumeMultiplier, PrimaryOperand=VariableAccessor.Float(Variable=@n1.Variable).Value).Out
    In 1 L=VariableAccessor.Audio(Variable=VariableMutator.Audio(Variable=@n2.Variable, Value=@WavePlayer_0.Out Left).Variable).Value
    In 0 L=VariableAccessor.Audio(Variable=VariableMutator.Audio(Variable=@n4.Variable, Value=@WavePlayer_1.Out Left).Variable).Value
    In 0 R=VariableAccessor.Audio(Variable=VariableMutator.Audio(Variable=@n8.Variable, Value=@WavePlayer_1.Out Right).Variable).Value
    In 1 R=VariableAccessor.Audio(Variable=VariableMutator.Audio(Variable=@n9.Variable, Value=@WavePlayer_0.Out Right).Variable).Value
   )
   InitVariable.Time()
   InitVariable.Time()
   InitVariable.Time()
   InitVariable.Time()
   InitVariable.Time()
   InitVariable.Trigger()
   InitVariable.Trigger()
   InitVariable.Trigger()
   InitVariable.Trigger()
   InitVariable.Trigger()
   n10 = InitVariable.Trigger()
   TriggerAnyTriggerAny2_1 = "TriggerAny.Trigger Any (2)"(In 0=UE.Reroute(Value=@n7.Is Not Valid).Value, In 1=@WavePlayer_1.On Finished)
   TriggerAccumulatorTriggerAccumulate2_0 = "TriggerAccumulator.Trigger Accumulate (2)"(
    In 0=VariableAccessor.Trigger(Variable=@n6.Variable).Value
    In 1=VariableAccessor.Trigger(Variable=VariableMutator.Trigger(Variable=@n10.Variable, Value=@TriggerAnyTriggerAny2_1.Out).Variable).Value
   )
   TriggerAnyTriggerAny3_0 = "TriggerAny.Trigger Any (3)"(In 0=@TriggerAccumulatorTriggerAccumulate2_0.Out)
  ]
  links [
   {dst="@UE.OutputFormat.Stereo.Audio:0", src=AudioMixerAudioMixerStereo2_0.Out L}
   {dst="@UE.OutputFormat.Stereo.Audio:1", src=AudioMixerAudioMixerStereo2_0.Out R}
  ]
 }]
properties {
 assetImportData=&self:AssetImportData
 audioPropertiesSheet=/BRCosmetics/Sounds/Settings/PropertySheets/Emotes/APS_Cosmetic_Emote_Music_1P
 bLooping=true
 duration=10000
 numChannels=2
 outputFormat=Stereo
 sampleRate=48000
 soundClassObject=/Game/Sounds/FortSoundClassesAndModes/Fort_Music_Emote_1P_PSM
 virtualizationMode=PlayWhenSilent
 concurrencySet [/Game/Sounds/Concurrency_Rules/Athena_EmoteMusic_1P_Conc]
 soundSubmixSends [{}]
}
